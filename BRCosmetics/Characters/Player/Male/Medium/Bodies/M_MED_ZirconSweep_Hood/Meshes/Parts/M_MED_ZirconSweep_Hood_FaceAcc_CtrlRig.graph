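# Control Rig for the ZirconSweep Hood face accessory that jitters the head-parented dyn_hair bones while the skydiving variable is true, likely as wind flutter.
# When active, it loops over a hard-coded bone array and applies a procedural rotation to each through Modify Transforms.
# A library function builds that rotation from per-axis noise within seeded random symmetric ranges, with speed scale setting the noise speed.
# The function takes no per-bone input.
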ControlRigBlueprint /BRCosmetics/Characters/Player/Male/Medium/Bodies/M_MED_ZirconSweep_Hood/Meshes/Parts/M_MED_ZirconSweep_Hood_FaceAcc_CtrlRig
graphs [
 New Function_ContainedGraph {nodes [
   n1 = RigVMFunctionEntryNode()
   Random = /Script/RigVM.RigVMFunction_RandomFloat(Duration=10, Maximum=15, Minimum=0, Seed=217) {is=struct RigVMUnitNode, template="Random::Execute(in Seed,in Minimum,in Maximum,in Duration,out Result)"}
   Multiply = &RigVMFunction_MathDoubleMul(B=-1, A=@Random.Result)
   Noise = /Script/RigVM.RigVMFunction_NoiseFloat(Frequency=0.5, Value=0.5, Speed=@n1.Scale, Minimum=@Multiply.Result, Maximum=@Random.Result) {is=struct RigVMUnitNode, template="Noise::Execute(in Value,in Speed,in Frequency,in Minimum,in Maximum,out Result)"}
   n2 = &RigVMFunction_RandomFloat(Duration=10, Maximum=20, Minimum=0, Seed=963) {template="Random::Execute(in Seed,in Minimum,in Maximum,in Duration,out Result)"}
   n3 = &RigVMFunction_RandomFloat(Duration=10, Maximum=15, Minimum=0, Seed=132) {template="Random::Execute(in Seed,in Minimum,in Maximum,in Duration,out Result)"}
   RigVMFunction_MathQuaternionFromEuler = &RigVMFunction_MathQuaternionFromEuler(
    Euler=(X=0,Y=0,Z=0)
    Euler.X=@Noise.Result
    Euler.Y=&RigVMFunction_NoiseFloat(Frequency=0.5, Value=0.3, Speed=@n1.Scale, Minimum=&RigVMFunction_MathDoubleMul(B=-1, A=@n2.Result).Result, Maximum=@n2.Result) {template="Noise::Execute(in Value,in Speed,in Frequency,in Minimum,in Maximum,out Result)"}.Result
    Euler.Z=&RigVMFunction_NoiseFloat(Frequency=0.5, Value=0.5, Speed=@n1.Scale, Minimum=&RigVMFunction_MathDoubleMul(B=-1, A=@n3.Result).Result, Maximum=@n3.Result) {template="Noise::Execute(in Value,in Speed,in Frequency,in Minimum,in Maximum,out Result)"}.Result
   )
   > RigVMFunctionReturnNode(Result=@RigVMFunction_MathQuaternionFromEuler.Result)
  ]}
 RigVMFunctionLibrary {nodes [New Function = RigVMCollapseNode(Scale=0)]}
 RigVMModel {nodes [
   RigUnit_BeginExecution = &RigUnit_BeginExecution()
   RigVMFunction_ControlFlowBranch = &RigVMFunction_ControlFlowBranch(ExecuteContext=@RigUnit_BeginExecution.ExecutePin, Condition=RigVMVariableNode() {variable=skydiving}.Value)
   RerouteNode = RigVMDispatchNode(Values="((Type=Bone,Name=\"dyn_hair_24\"),((),Type=Bone,Name=\"dyn_hair_25\"),((),Type=Bone,Name=\"dyn_hair_26\"),((),Type=Bone,Name=\"dyn_hair_27\"),((),Type=Bone,Name=\"dyn_hair_28\"),((),Type=Bone,Name=\"dyn_hair_32\"),((),Type=Bone,Name=\"dyn_hair_33\"),((),Type=Bone,Name=\" ...(truncated)") {template="DISPATCH_RigVMDispatch_ArrayMake(in Values,out Array)"}
   DISPATCH_RigVMDispatch_ArrayIterator = RigVMDispatchNode(Array=@RerouteNode.Array, ExecuteContext=@RigVMFunction_ControlFlowBranch.True) {template="DISPATCH_RigVMDispatch_ArrayIterator(in Array,out Element,out Index,out Count,out Ratio)"}
   New Function = RigVMFunctionReferenceNode(Scale=RigVMVariableNode() {variable=speed scale}.Value)
   > &RigUnit_ModifyTransforms(
    ItemToModify=((Item=(Type=Bone),Transform=(Rotation=(X=0,Y=0,Z=0,W=1),Translation=(X=0,Y=0,Z=0),Scale3D=(X=1,Y=1,Z=1))))
    Weight=1
    ItemToModify.0.Item=@DISPATCH_RigVMDispatch_ArrayIterator.Element
    ItemToModify.0.Transform.Rotation=@"New Function".Result
   )
  ]}
]
hierarchy [
 root {loc=<0 0 0>, rot=<0 0 0 1>}
 pelvis {parent=root, loc=<0 -0.7071 93.1191>, rot=<0.0127 0.707 -0.0127 -0.707>}
 spine_01 {parent=pelvis, loc=<6.7494 -0.6388 0>, rot=<0 0 0.034 -0.9994>}
 spine_02 {parent=spine_01, loc=<7.2518 -0.035 0>, rot=<0 0 0.0238 -0.9997>}
 spine_03 {parent=spine_02, loc=<6.0104 0 0>, rot=<0 0 0.0149 -0.9999>}
 spine_04 {parent=spine_03, loc=<8.1024 0 0>, rot=<0 0 -0.0132 -0.9999>}
 spine_05 {parent=spine_04, loc=<10.0146 0 0>, rot=<0 0 -0.0015 -1>}
 neck_01 {parent=spine_05, loc=<11.3138 0.1529 0>, rot=<0 0 -0.2003 -0.9797>}
 neck_02 {parent=neck_01, loc=<4.5627 0 0>, rot=<0 0 0.0504 -0.9987>}
 head {parent=neck_02, loc=<5.3914 0 0>, rot=<0 0 0.115 -0.9934>}
 dyn_hair_40 {parent=head, loc=<8.9808 6.7658 -6.0842>, rot=<0.0877 -0.8954 0.3911 -0.1936>}
 dyn_hair_41 {parent=head, loc=<12.9937 6.6618 7.1473>, rot=<0.8341 0.0413 0.5472 0.0555>}
 dyn_hair_25 {parent=head, loc=<8.8807 10.1856 -2.4838>, rot=<0.3174 0.6261 0.6414 0.3097>}
 dyn_hair_24 {parent=head, loc=<8.9827 10.2564 -0.657>, rot=<0.1028 0.2474 0.9632 0.0227>}
 dyn_hair_27 {parent=head, loc=<7.6688 10.0526 4.2664>, rot=<0.1684 0.1368 0.976 0.0195>}
 dyn_hair_28 {parent=head, loc=<12.2766 9.7616 3.6532>, rot=<0.5166 0.2878 0.7848 0.1854>}
 dyn_hair_32 {parent=head, loc=<10.5112 9.0139 6.0206>, rot=<0.3376 0.2246 0.9107 0.0792>}
 dyn_hair_33 {parent=head, loc=<11.5673 8.2779 7.2376>, rot=<0.6896 0.1523 0.6927 0.1462>}
 dyn_hair_26 {parent=head, loc=<8.0921 8.403 -4.9273>, rot=<0.056 -0.9166 0.3742 -0.1292>}
 dyn_hair_35 {parent=head, loc=<12.6946 8.04 -4.796>, rot=<-0.15 -0.834 -0.2876 -0.4463>}
 dyn_hair_36 {parent=head, loc=<13.4731 6.758 -5.7876>, rot=<-0.0535 -0.694 -0.0576 -0.7157>}
 dyn_hair_37 {parent=head, loc=<14.6998 5.1405 -4.9415>, rot=<0.1238 -0.5981 0.0891 -0.7868>}
 dyn_hair_38 {parent=head, loc=<14.4662 8.5391 3.5331>, rot=<0.7955 0.1703 0.5239 0.2522>}
 dyn_hair_39 {parent=head, loc=<8.2354 8.195 6.1585>, rot=<0.1761 -0.4326 0.8797 -0.0895>}
]
variables [
 skydiving {type=bool, default=False}
 speed scale {type=double, default=0}
]
